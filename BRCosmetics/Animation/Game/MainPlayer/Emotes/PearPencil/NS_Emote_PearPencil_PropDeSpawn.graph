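# Niagara system for the despawn effect of the prop in the PearPencil player emote.
# It combines three emitters named for sparkles, airglow and smoke, though no renderers are listed for them here.
# System-level logic uses SystemState and a variable-assignment module, and the system uses a skeletal mesh data interface.
# It exposes a NormalizedNotifyProgress float user parameter, which by its name receives animation notify progress.
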NiagaraSystem /BRCosmetics/Animation/Game/MainPlayer/Emotes/PearPencil/NS_Emote_PearPencil_PropDeSpawn
dependencies {
 usedDataInterfaces [&NiagaraDataInterfaceSkeletalMesh]
 usedDynamicInputs []
 usedModules [
  &self:SystemSpawnScript.SystemScriptSource.SystemScriptGraph.NiagaraNodeAssignment_1.SetVariables_#1
  &SystemState
 ]
}
emitters {
 AirglowLWE002 {rendererClasses [], renderers []}
 SmokeLWE002 {rendererClasses [], renderers []}
 SparklesLWE002 {rendererClasses [], renderers []}
}
summary {
 systemName=NS_Emote_PearPencil_PropDeSpawn
 emitters [
  {emitterName=SparklesLWE002, rendererClasses []}
  {emitterName=AirglowLWE002, rendererClasses []}
  {emitterName=SmokeLWE002, rendererClasses []}
 ]
 userVariables [User.NormalizedNotifyProgress {type=NiagaraFloat, defaultValue=0, description=""}]
}
